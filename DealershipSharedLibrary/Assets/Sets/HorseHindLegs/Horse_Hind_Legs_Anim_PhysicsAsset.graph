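# Physics asset for the Horse_Hind_Legs_Anim skeletal mesh, giving the horse a simplified collision and ragdoll body.
# Despite the hind-legs name, single capsules cover selected bones across the whole animal: root, spine, neck, head, tail and two per leg.
# Constraints form a tree from Skeleton_Root to the spine, tail and back legs, with neck and front legs branching from Spine3.
# Every joint shares identical swing and twist limit angles, with no per-joint tuning.
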PhysicsAsset /DealershipSharedLibrary/Assets/Sets/HorseHindLegs/Horse_Hind_Legs_Anim_PhysicsAsset
bodies {
 Leg_Back_L2 [{shapeType=Capsule, length=14.0924, radius=11.5548, center=<-12.9543 -6.6812 2.3728>, rotation=<rot 0 -52.538 -71.2999>}]
 Leg_Back_L4 [{shapeType=Capsule, length=12.3335, radius=4.0795, center=<-11.8238 2.0367 -0.9798>, rotation=<rot 0 81.2753 -84.847>}]
 Leg_Back_R2 [{shapeType=Capsule, length=13.9719, radius=11.9432, center=<13.3203 6.2763 -2.9231>, rotation=<rot 0 -57.0038 -76.7848>}]
 Leg_Back_R4 [{shapeType=Capsule, length=12.3833, radius=4.0294, center=<11.8909 -1.9682 1.0389>, rotation=<rot 0 81.9654 -84.5536>}]
 Leg_Front_L2 [{shapeType=Capsule, length=14.387, radius=7.3727, center=<-12.1213 1.3087 0.3243>, rotation=<rot 0 -100.2199 -87.0382>}]
 Leg_Front_L4 [{shapeType=Capsule, length=11.1503, radius=4.3253, center=<-11.0873 2.1039 0.0926>, rotation=<rot 0 -108.8501 -89.4108>}]
 Leg_Front_R2 [{shapeType=Capsule, length=14.1081, radius=7.4366, center=<12.1921 -1.9355 -0.1355>, rotation=<rot 0 -103.9524 -82.1508>}]
 Leg_Front_R4 [{shapeType=Capsule, length=11.1506, radius=4.3257, center=<11.0873 -2.1033 -0.0898>, rotation=<rot 0 -108.8509 -89.4125>}]
 Neck2 [{shapeType=Capsule, length=13.0958, radius=10.5053, center=<6.5278 -2.8126 -0.5683>, rotation=<rot 0 -126.3497 -85.7638>}]
 Skeleton_Root [{shapeType=Capsule, length=22.7225, radius=14.2714, center=<2.6546 7.8853 0.0062>, rotation=<rot 0 -106.0581 -89.8734>}]
 Spine2 [{shapeType=Capsule, length=17.7999, radius=13.7738, center=<6.3698 10.4236 -0.0949>, rotation=<rot 0 -30.9099 -87.6487>}]
 Spine3 [{shapeType=Capsule, length=15.7503, radius=12.7379, center=<11.376 2.5965 -0.1089>, rotation=<rot 0 141.1425 -86.982>}]
 Tail4 [{shapeType=Capsule, length=13.4088, radius=5.5509, center=<13.5207 1.4426 -0.1631>, rotation=<rot 0 96.7463 -89.8871>}]
 head [{shapeType=Capsule, length=15.0792, radius=7.1746, center=<6.264 0.0284 0.0029>, rotation=<rot 0 108.1373 -89.8581>}]
}
constraints [
 {bone1Name=Spine2, bone2Name=Skeleton_Root, swing1LimitDegrees=45, swing2LimitDegrees=45, twistLimitDegrees=45}
 {bone1Name=Spine3, bone2Name=Spine2, swing1LimitDegrees=45, swing2LimitDegrees=45, twistLimitDegrees=45}
 {bone1Name=Neck2, bone2Name=Spine3, swing1LimitDegrees=45, swing2LimitDegrees=45, twistLimitDegrees=45}
 {bone1Name=head, bone2Name=Neck2, swing1LimitDegrees=45, swing2LimitDegrees=45, twistLimitDegrees=45}
 {bone1Name=Leg_Front_R2, bone2Name=Spine3, swing1LimitDegrees=45, swing2LimitDegrees=45, twistLimitDegrees=45}
 {bone1Name=Leg_Front_R4, bone2Name=Leg_Front_R2, swing1LimitDegrees=45, swing2LimitDegrees=45, twistLimitDegrees=45}
 {bone1Name=Leg_Front_L2, bone2Name=Spine3, swing1LimitDegrees=45, swing2LimitDegrees=45, twistLimitDegrees=45}
 {bone1Name=Leg_Front_L4, bone2Name=Leg_Front_L2, swing1LimitDegrees=45, swing2LimitDegrees=45, twistLimitDegrees=45}
 {bone1Name=Tail4, bone2Name=Skeleton_Root, swing1LimitDegrees=45, swing2LimitDegrees=45, twistLimitDegrees=45}
 {bone1Name=Leg_Back_R2, bone2Name=Skeleton_Root, swing1LimitDegrees=45, swing2LimitDegrees=45, twistLimitDegrees=45}
 {bone1Name=Leg_Back_R4, bone2Name=Leg_Back_R2, swing1LimitDegrees=45, swing2LimitDegrees=45, twistLimitDegrees=45}
 {bone1Name=Leg_Back_L2, bone2Name=Skeleton_Root, swing1LimitDegrees=45, swing2LimitDegrees=45, twistLimitDegrees=45}
 {bone1Name=Leg_Back_L4, bone2Name=Leg_Back_L2, swing1LimitDegrees=45, swing2LimitDegrees=45, twistLimitDegrees=45}
]
